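# Parameter container material for a Zest vehicle cosmetic attachment, apparently for Mutable, enabled for skeletal meshes and mesh deformers.
# It declares BC, Mask and Normal textures with blank defaults, surface scalars, and wind animation parameters, some marked as code-driven.
# All parameters are chained through Multiply nodes into a final node that connects to no material output, so the graph holds parameters without shading or animating anything.
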Material /VehicleCosmetics/PassthroughMeshes/Zest/M_Zest_Att_MutableContainer
graph {
 properties {bAutomaticallySetUsageInEditor=false, bUsedWithMeshDeformer=true, bUsedWithSkeletalMesh=true, twoSided=true, usedShadingModels=Default Lit}
 nodes [
  n1 = TextureSampleParameter2D() {
   group=1_Input Textures
   parameterName=BC
   samplerSource=SSM_Wrap_WorldGroupSettings
   sortPriority=0
   texture=/VehicleCosmetics/Textures/T_Black
  }
  n2 = TextureSampleParameter2D() {
   group=1_Input Textures
   parameterName=Normal
   samplerSource=SSM_Wrap_WorldGroupSettings
   samplerType=SAMPLERTYPE_Normal
   sortPriority=2
   texture=/VehicleCosmetics/Textures/T_BlankNormal
  }
  n3 = Multiply(A=@n1.RGB, B=@n2.RGB)
  n4 = TextureSampleParameter2D() {
   group=1_Input Textures
   parameterName=Mask
   samplerSource=SSM_Wrap_WorldGroupSettings
   samplerType=SAMPLERTYPE_Masks
   sortPriority=1
   texture=/VehicleCosmetics/Textures/T_Black_Masks
   channelNames {a=Unassigned, b=Unassigned, g=Rough, r=Metallic}
  }
  n5 = Multiply(
   A=Multiply(
    A=Multiply(A=@n3, B=@n4.RGB)
    B=ScalarParameter() {defaultValue=2, desc=Darken or Lighten BC, group=2_Material Properties, parameterName=BCMultiply}
   )
   B=ScalarParameter() {defaultValue=0.2, desc=Offset for mirrored image., group=2_Material Properties, parameterName=MirrorTranslation}
  )
  n6 = ScalarParameter() {defaultValue=0.5, group=2_Material Properties, parameterName=Specular, sortPriority=5}
  n7 = ScalarParameter() {defaultValue=1.5, desc="Wind speed - do not change, should only be altered via code.", group=3_WPO Animation, parameterName=WindSpeed}
  n8 = ScalarParameter() {defaultValue=0.5, desc="Wind strength - do not change, should only be altered via code.", group=3_WPO Animation, parameterName=WindInt}
  n9 = Multiply(A=Multiply(A=Multiply(A=@n5, B=@n6), B=@n7), B=@n8)
  n10 = VectorParameter() {desc=Do not change - will be altered via code., group=3_WPO Animation, parameterName=WindDirectionXYZ-RGB, defaultValue=<rgba 1 1 0 1>}
  n11 = ScalarParameter() {defaultValue=1, desc="Animation strength. Higher numbers == more extreme animation.", group=3_WPO Animation, parameterName=Animation}
  Multiply(B=Multiply(A=Multiply(A=@n9, B=@n10.RGB), B=@n11))
 ]
}
