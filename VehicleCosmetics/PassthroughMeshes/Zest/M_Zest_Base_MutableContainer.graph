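# M_Zest_Base_MutableContainer is a minimal Default Lit material in the VehicleCosmetics passthrough mesh folder for Zest.
# Its usage flags enable skeletal mesh and mesh deformer rendering.
# The graph holds a single constant node and no textures or parameters.
# The name suggests it serves as a placeholder base material for a Mutable customization container.
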Material /VehicleCosmetics/PassthroughMeshes/Zest/M_Zest_Base_MutableContainer
graph {properties {bUsedWithMeshDeformer=true, bUsedWithSkeletalMesh=true, usedShadingModels=Default Lit}, nodes [Constant() {r=0.25}]}
