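# Physics asset for the Zest attachment skeletal mesh among the vehicle cosmetics passthrough meshes, named as a Mutable container variant.
# It holds a single body on the dyn_zest_01 bone, shaped as one offset, rotated capsule.
# It defines no constraints, so the body is not jointed to any other within this asset.
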PhysicsAsset /VehicleCosmetics/PassthroughMeshes/Zest/SK_Zest_Att_PhysicsAsset_MutableContainer
bodies {dyn_zest_01 [{shapeType=Capsule, length=30.4992, radius=24.4806, center=<18.7414 -0.0117 -24.0344>, rotation=<rot 0.0278 -90 16.0209>}]}
constraints []
