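# Physics asset for the Zest base skeletal mesh among the vehicle cosmetics passthrough meshes.
# Its constraint list is empty, so it defines no joints between physics bodies.
# The MutableContainer suffix suggests it serves as a container for the Mutable customization system.
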PhysicsAsset /VehicleCosmetics/PassthroughMeshes/Zest/SK_Zest_Base_PhysicsAsset_MutableContainer
constraints []
